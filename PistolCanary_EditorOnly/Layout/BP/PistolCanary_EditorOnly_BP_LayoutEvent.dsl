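; Editor-only layout trigger that previews the PistolCanary event's level sequence when a pawn walks through its slate mesh.
; Overlap hides the slate and, if armed, spawns a level sequence actor and plays it, optionally delayed.
; Stopping re-shows the slate and re-arms the trigger.
; Custom events stop, toggle pause with a frame printout, jump to a marker or start, and fire timed "ke" console commands for day-by-day lead-up.
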(event OnComponentBeginOverlap(walkThroughThis)
  (params (OverlappedComponent :type Primitive%20Component%20Object%20Reference) (OtherActor :type Actor%20Object%20Reference) (OtherComp :type Primitive%20Component%20Object%20Reference) (OtherBodyIndex :type Integer) (bFromSweep :type Boolean) (SweepResult :type Hit%20Result%20Structure%20%28by%20ref%29))
  ; pawn walks through a mesh - turns off vis; if trigger was intentional, plays event
  (Rendering|SetVisibility (Variables|Default|GetwalkThroughThis) false false)
  (if (Variables|Default|GetplaySeq)
    (CallFunction|LayoutPistolCanary)))

(event Custom|layoutPistolCanary_leadup
  (bind _wwait (Variables|Default|GetWwait))
  (Development|ExecuteConsoleCommand "cheat ke * PC_Start_Day1")
  (Development|ExecuteConsoleCommand "cheat ke * PC_Day1")
  (Utilities|FlowControl|Delay _wwait)
  (Development|ExecuteConsoleCommand "cheat ke * PC_Start_Day2")
  (Development|ExecuteConsoleCommand "cheat ke * PC_Day2")
  (Utilities|FlowControl|Delay _wwait)
  (Development|ExecuteConsoleCommand "cheat ke * PC_Start_Day3")
  (Development|ExecuteConsoleCommand "cheat ke * PC_Event"))

(event Custom|layoutPistolCanary
  ; play event - spawns replicated level sequence, option to wait a couple seconds before play. unhides slate when sequence is stopped/finished so event can be triggered again
  (Variables|Default|SetplaySeq false)
  (bind _returnvalue (Game|SpawnActorfromClass :Class "/Script/LevelSequence.ReplicatedLevelSequenceActor" :SpawnTransform (Math|Transform|MakeTransform "0, 0, 0") :CollisionHandlingOverride "AlwaysSpawn" :TransformScaleMethod "MultiplyWithRoot" :PlaybackSettings (Utilities|Struct|MakeMovieSceneSequencePlaybackSettings false "(Value=0)" "(TickIntervalSeconds=0.000000,EvaluationBudgetMicroseconds=0.000000,bTickWhenPaused=False,bAllowRounding=True)" 1.0 0.0 false false false false false (Variables|Default|GetdisableCameraCuts)) :LevelSequenceAsset (Variables|Default|GetPlayThis) :bReplicatePlayback false))
  (if (not (Variables|Default|GetplayInstantly))
    (Utilities|FlowControl|Delay (Variables|Default|GetsecondsBeforePlay)))
  (bind _output_get (Variables|Default|SetlayoutSeqActor _returnvalue))
  (Sequencer|Player|Play (Variables|/Script/LevelSequence.LevelSequenceActor|GetSequencePlayer _output_get) :!class /Script/MovieScene.MovieSceneSequencePlayer)
  (Sequencer|Player|BindEventtoOnStop (Variables|/Script/LevelSequence.LevelSequenceActor|GetSequencePlayer _output_get) (delegate Custom|showSlate)))

(event Custom|showSlate
  ; play event - spawns replicated level sequence, option to wait a couple seconds before play. unhides slate when sequence is stopped/finished so event can be triggered again
  (Rendering|SetVisibility (Variables|Default|GetwalkThroughThis) true false)
  (Variables|Default|SetplaySeq true))

(event Custom|layoutPistolCanaryStop
  ; stop event - stops event at current time and unhides slate
  (Sequencer|Player|StopatCurrentTime (Variables|/Script/LevelSequence.LevelSequenceActor|GetSequencePlayer (Variables|Default|GetlayoutSeqActor)) :!class /Script/MovieScene.MovieSceneSequencePlayer)
  (CallFunction|ShowSlate))

(event Custom|layoutPistolCanaryPause
  (bind _layoutseqactor (Variables|Default|GetlayoutSeqActor))
  (bind _sequenceplayer (Variables|/Script/LevelSequence.LevelSequenceActor|GetSequencePlayer _layoutseqactor))
  (bind _returnvalue_frame (Utilities|TimeManagement|BreakQualifiedFrameTime (Sequencer|Player|GetCurrentTime _sequenceplayer)))
  (bind _returnvalue_frame_value (Utilities|Struct|BreakFrameNumber _returnvalue_frame))
  ; pause event - pause event at current time, unpause if paused
  (if (Sequencer|Player|IsPaused _sequenceplayer :!class /Script/MovieScene.MovieSceneSequencePlayer)
    (Sequencer|Player|Play _sequenceplayer :!class /Script/MovieScene.MovieSceneSequencePlayer)
    (else
      (Sequencer|Player|Pause _sequenceplayer)
      (Development|PrintString (Utilities|String|Append (Sequencer|Player|GetSequenceName _sequenceplayer false :!class /Script/MovieScene.MovieSceneSequencePlayer) "   :   Paused on frame number " (Utilities|String|ToString(Integer) _returnvalue_frame_value)) true true "(R=0.162462,G=0.046959,B=0.560764,A=1.000000)" 3.0))))

(event Custom|layoutPistolCanaryrdh
  ; sequencer marker - jump to sequences or beginning
  (Sequencer|Player|SetPlaybackPosition (Variables|/Script/LevelSequence.LevelSequenceActor|GetSequencePlayer (Variables|Default|GetlayoutSeqActor)) (Utilities|Struct|MakeMovieSceneSequencePlaybackParams "(FrameNumber=(Value=0),SubFrame=0.000000)" 0.0 (Variables|Default|GetcustomMarker) "(Hours=0,Minutes=0,Seconds=0,Frames=0,Subframe=0.000000,bDropFrameFormat=False)" "MarkedFrame" "Jump")))

(event Custom|layoutPistolCanaryStartFrame
  ; sequencer marker - jump to sequences or beginning
  (Sequencer|Player|SetPlaybackPosition (Variables|/Script/LevelSequence.LevelSequenceActor|GetSequencePlayer (Variables|Default|GetlayoutSeqActor)) (struct :Frame (struct :FrameNumber (struct :Value 0) :SubFrame 0.0) :Time 0.0 :MarkedFrame "" :Timecode "(Hours=0,Minutes=0,Seconds=0,Frames=0,Subframe=0.000000,bDropFrameFormat=False)" :PositionType "Frame" :UpdateMethod "Jump" :bHasJumped false)))

(event Custom|layoutPistolCanaryrdhPause)

(fn ConstructionScript)
